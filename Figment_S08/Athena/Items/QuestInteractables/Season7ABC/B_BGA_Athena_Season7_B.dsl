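; Season 7 quest interactable actor that decides whether a pawn may interact with it.
; BlueprintCanInteract gets the pawn's player controller, checks quest and objective completion, and requires the actor to be unlit and unblocked.
; A valid, unfinished objective plus the pawn's required gameplay tags allows interaction; otherwise the actor must be both unused and hotfix-enabled.
; Begin play, overlap and tick events contain no logic of their own.
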(event EventBeginPlay
  (|CallParentFunction) ; This node is disabled and will not be called.\nDrag off pins to build functionality.
)

(event Collision|EventActorBeginOverlap
  (params (OtherActor :type Actor%20Object%20Reference)))

(event EventTick
  (params (DeltaSeconds :type Float%20%28single-precision%29)))

(fn BlueprintCanInteract
  (params (InteractingPawn :type Fort%20Pawn%20Object%20Reference) (InteractionBeingAttempted :type EInteractionBeingAttempted%20Enum :default "FirstInteraction") (InteractionType :type TInteractionType%20Enum :default "IT_NoInteraction"))
  (bind _returnvalue (Fortnite|GetFortPlayerControllerfromActor InteractingPawn))
  (Utilities|IsValid _returnvalue
    (:Is%20Valid
      (bind (_questisvalid _questcompleted _objectivecompleted) (Default|AthenaCheckQuestandObjectiveCompletion _returnvalue (Variables|Default|GetSeason7QuestItem) (Variables|Default|GetS7BackendObjName)))
      (return (Math|Boolean|ANDBoolean (!= (Variables|Default|GetLitRepBool) 1) (or (not (Variables|Default|GetBeenUsed)) (Math|Boolean|ANDBoolean _questisvalid (not _objectivecompleted) (GameplayTags|HasAllTags (GameplayTags|GetOwnedGameplayTags InteractingPawn :!class /Script/GameplayTags.GameplayTagAssetInterface) (Variables|QuestStuff|GetCustomSpecialS7Tags) true))) (or (Math|Boolean|ANDBoolean _questisvalid (not _objectivecompleted) (GameplayTags|HasAllTags (GameplayTags|GetOwnedGameplayTags InteractingPawn :!class /Script/GameplayTags.GameplayTagAssetInterface) (Variables|QuestStuff|GetCustomSpecialS7Tags) true)) (Variables|Default|GetHotfixedEnabled)) (not (Variables|Default|GetBlockInteraction)))))))

(fn ConstructionScript
  (|CallParentFunction))
